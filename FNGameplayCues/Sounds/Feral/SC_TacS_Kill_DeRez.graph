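# Sound cue for a de-rez kill sound in the Feral gameplay cue set.
# A Random node picks one of three Kill_Derez wave players with equal weight for variation.
# It uses a teleporter gadget attenuation asset, overrides concurrency, and plays in the gadgets sound class.
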SoundCue /FNGameplayCues/Sounds/Feral/SC_TacS_Kill_DeRez
root=Random_1
nodes [
 {id=Random_1, children [WavePlayer_4,WavePlayer_5,WavePlayer_6], properties {weights [1,1,1]}}
 {id=WavePlayer_4, properties {soundWaveAssetPtr=/FNGameplayCues/Sounds/Feral/Gunplay_Update/Kill_Derez_01}}
 {id=WavePlayer_5, properties {soundWaveAssetPtr=/FNGameplayCues/Sounds/Feral/Gunplay_Update/Kill_Derez_02}}
 {id=WavePlayer_6, properties {soundWaveAssetPtr=/FNGameplayCues/Sounds/Feral/Gunplay_Update/Kill_Derez_03}}
]
properties {
 attenuationSettings=/Game/Sounds/Attenuation/Fort_TeleporterGadget_Rez_Target_Att
 bOverrideConcurrency=true
 maxDistance=12000
 soundClassObject=&Fort_Gadgets
}
